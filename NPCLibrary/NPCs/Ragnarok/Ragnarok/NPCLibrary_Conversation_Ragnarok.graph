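# Conversation database for the Ragnarok NPC, providing his greeting and access to shared NPC dialogue.
# The entry point leads to one Speech node with a default line and contextual alternatives for first-ever meeting, visiting the Viking ship POI, and wearing the Golden Skeleton outfit.
# Requirements gate each alternative, with priorities drawn from Tandem.Speech.Priority curve rows.
# The speech node then outputs to link nodes that jump to the shared NPCLibrary Quests and BasicServices conversations.
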ConversationDatabase /NPCLibrary/NPCs/Ragnarok/Ragnarok/NPCLibrary_Conversation_Ragnarok
entry_points [{tag=Conversation.Entry.Ragnarok, destinations [#1]}]
nodes [
 ConversationEntryPointNode_7 {
  id=#1
  class=ConversationEntryPointNode
  outputs [#2]
  properties {entryTag {tagName=Conversation.Entry.Ragnarok}}
 }
 Speech_C_4 {
  id=#2
  class=Speech_C
  outputs [#3, #4]
  properties {
   nodeName=""
   generalConfig {
    defaultMessage="I await my destiny. "
    contextualMessages [
     {
      message="I <b>knew</> you'd come eventually."
      contextRequirements [{requirement {class=FortControllerRequirement_IsFirstEverConversationWithNPC}}]
      priority {curve {rowName=Tandem.Speech.Priority.FirstEverInteraction}}
     }
     {
      message=" Did Blackheart say anything about me when you stopped by his vessel? Just... <b>wondering.</>"
      contextRequirements [{requirement {class=FortControllerRequirement_HasVisitedPOI, properties {pOITag {tagName=Athena.Location.UnNamedPOI.VikingShip}}}}]
      priority {curve {rowName=Tandem.Speech.Priority.VisitedSpecificPOI}}
     }
     {
      message=" Looking like you're in rough shape there, brother."
      contextRequirements [{requirement {
         class=FortControllerRequirement_HasCID
         properties {allowedCIds [/BRCosmetics/Athena/Items/Cosmetics/Characters/CID_675_Athena_Commando_M_TheGoldenSkeleton]}
        }}]
      priority {curve {rowName=Tandem.Speech.Priority.OutfitReaction}}
     }
    ]
   }
  }
 }
 ConversationLinkNode_0 {id=#4, class=ConversationLinkNode, properties {remoteEntryTag {tagName=Conversation.Entry.BasicServices}}}
 ConversationLinkNode_1 {id=#3, class=ConversationLinkNode, properties {remoteEntryTag {tagName=Conversation.Entry.NPCLibrary.Quests}}}
]
